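# M_Distortion_Wave2 is an unlit translucent Niagara sprite material for explosion distortion waves, using index-of-refraction refraction.
# A lerp blends from one toward the inverted green channel of a soft flare mask by particle alpha, so the mask's effect fades with the particle.
# A pointy normal map passes through FlattenNormal at low flatness, and spherical particle normals are enabled.
# Material output connections are absent from the graph.
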Material /BRCosmetics/Effects/Fort_Effects/Effects/Explosion/M_Distortion_Wave2
graph {
 properties {
  bGenerateSphericalParticleNormals=true
  bUsedWithNiagaraSprites=true
  blendMode=BLEND_Translucent
  pixelDepthOffsetMode=PDOM_Legacy
  refractionMethod=RM_IndexOfRefraction
  shadingModel=MSM_Unlit
  usedShadingModels=Unlit
 }
 nodes [
  Constant() {r=2}
  Constant()
  LinearInterpolate(B=OneMinus(None=TextureSample() {texture=/Game/Global/Textures/Default/Masks_and_Gradients/T_New_Soft_Flair}.G), Alpha=ParticleColor().A) {constA=1}
  MaterialFunctionCall(
   Normal=TextureSample() {samplerType=SAMPLERTYPE_Normal, texture=/Game/Effects/Fort_Effects/Textures/PointyNormal}.RGB
   Flatness=Constant() {r=0.1}
  ) {materialFunction=&FlattenNormal}
 ]
}
